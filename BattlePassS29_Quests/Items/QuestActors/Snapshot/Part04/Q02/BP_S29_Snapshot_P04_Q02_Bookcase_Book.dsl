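; Season 29 Snapshot quest book prop that slides a referenced bookcase door open when a player interacts with it.
; A timeline lerps the door over a fixed offset, with the authority toggling bDoorOpen on finish; interaction is blocked while open.
; Success on the linked Harpe quest prop reverses the slide.
; On closing, OnRep_bDoorOpen restores the book's visuals, effects and map icon for clients whose local player lacks that objective.
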(event EventEndPlay
  (params (EndPlayReason :type EEndPlayReason%20Enum :default "Destroyed"))
  (|CallParentFunction EndPlayReason)
  (Quest|UnbindEventfromOnQuestComponentSuccessEvent (Variables|/Script/FortniteGame.BuildingGameplayActorPropQuest|GetConsolidatedQuestComponent (Variables|Default|GetHarpeQuestProp)) (EventDispatchers|CreateEvent :!event Custom|OnHarpeCollected)))

(event EventBeginPlay
  (bind _doorbpreference (Variables|Default|GetDoorBPReference))
  (bind _harpequestprop (Variables|Default|GetHarpeQuestProp))
  ;[0 0 0] Set starting location of the door
  (|CallParentFunction)
  (Utilities|IsValid _doorbpreference
    (:Is%20Valid
      (bind _output_get (Variables|Default|SetDoorStartLocation (Transformation|GetActorLocation _doorbpreference)))
      (Variables|Default|SetDoorEndLocation (- _output_get "512.000000,0.000000,0.000000"))
      (Utilities|IsValid _harpequestprop
        (:Is%20Valid
          (Quest|BindEventtoOnQuestComponentSuccessEvent (Variables|/Script/FortniteGame.BuildingGameplayActorPropQuest|GetConsolidatedQuestComponent _harpequestprop) (EventDispatchers|CreateEvent :!event Custom|OnHarpeCollected)))))))

(event Interaction|EventBlueprintOnInteract
  (params (InteractingPawn :type Fort%20Pawn%20Object%20Reference) (InteractionBeingAttempted :type EInteractionBeingAttempted%20Enum :default "FirstInteraction"))
  ;[0 0 0] When prop is interacted with, slide the door over
  (|CallParentFunction InteractingPawn InteractionBeingAttempted)
  (if (not (Variables|Default|GetbDoorOpen))
    (Utilities|IsValid (Fortnite|GetFortPlayerControllerfromActor InteractingPawn)
      (:Is%20Valid
        (CallFunction|ExecuteDoorSlide false)))))

(event Custom|ExecuteDoorSlide
  (params (bReverseDirection :type Boolean))
  ;[0 0 0] Slide the door over on a timeline
  (if (not bReverseDirection)
    (Networking|SwitchHasAuthority
      (:Authority
        (bind (_direction _doortracklerp) (Timeline|Timeline PlayFromStart)
          (:Update
            (Transformation|SetActorRelativeLocation (Variables|Default|GetDoorBPReference) (Math|Vector|Lerp(Vector) (Variables|Default|GetDoorStartLocation) (Variables|Default|GetDoorEndLocation) _doortracklerp) false false))
          (:Finished
            (Networking|SwitchHasAuthority
              (:Authority
                (Variables|Default|SetbDoorOpen (not (Variables|Default|GetbDoorOpen))))))))
      (:Remote
        (VisualState|VFX&Audio|PostInteractionVFX&Audio (Fortnite|GetFirstLocalFortPlayerController true))
        (MiniMapData|SetVisibility (Variables|BuildingGameplayActorPropQuest|GetQuestIconComponent) false)
        (Timeline|Timeline PlayFromStart)))
    (else
      (Timeline|Timeline ReverseFromEnd))))

(event Custom|OnHarpeCollected
  (params (PlayerController :type Fort%20Player%20Controller%20Object%20Reference))
  (if (Variables|Default|GetbDoorOpen)
    ;[0 0 0] Slide the door over on a timeline
    (CallFunction|ExecuteDoorSlide true)))

(event OnQuestComponentSuccessEvent(ConsolidatedQuestComponent)
  (params (PlayerController :type Fort%20Player%20Controller%20Object%20Reference)))

(fn BlueprintCanInteract
  (params (InteractingPawn :type Fort%20Pawn%20Object%20Reference) (InteractionBeingAttempted :type EInteractionBeingAttempted%20Enum :default "FirstInteraction") (InteractionType :type TInteractionType%20Enum :default "IT_NoInteraction"))
  (if (Variables|Default|GetbDoorOpen)
    (return false)
    (else
      (return (|CallParentFunction InteractingPawn InteractionBeingAttempted InteractionType)))))

(fn IsInstigatorLocallyControlled
  (params (InstigatingPlayerController :type Fort%20Player%20Controller%20Object%20Reference))
  ;[0 0 0] Returns true on the executing client if the instigator controller is locally controlled. Returns false if running on the server
  (if (Networking|IsServer)
    (return false)
    (else
      (Utilities|IsValid InstigatingPlayerController
        (:Is%20Valid
          (if (PlayerController|IsinSplitscreen InstigatingPlayerController)
            (for _array_element (Fortnite|GetLocalFortPlayerControllers true)
              (if (== _array_element InstigatingPlayerController)
                (return true)))
            (return false)
            (else
              (return true))))))))

(fn OnRep_bDoorOpen
  (if (not (Variables|Default|GetbDoorOpen))
    (Networking|SwitchHasAuthority
      (:Remote
        (bind (_playercontrollerout _completedquestobjective) (Quest|HasPlayerCompletedQuestObjective (Fortnite|GetFirstLocalFortPlayerController true) (Class|ConsolidatedQuestComponent|GetQuestItemDefinition (Variables|/Script/FortniteGame.BuildingGameplayActorPropQuest|GetConsolidatedQuestComponent (Variables|Default|GetHarpeQuestProp))) (Variables|/Script/FortniteGame.ConsolidatedQuestComponent|GetObjectiveBackendName (Variables|/Script/FortniteGame.BuildingGameplayActorPropQuest|GetConsolidatedQuestComponent (Variables|Default|GetHarpeQuestProp)))))
        (if (not _completedquestobjective)
          (VisualState|VFX&Audio|PreInteractionVFX&Audio _playercontrollerout)
          (VisualState|Meshes|PreInteractionVisuals 0)
          (MiniMapData|SetVisibility (Variables|BuildingGameplayActorPropQuest|GetQuestIconComponent) true))))))

(fn ConstructionScript
  (|CallParentFunction))
