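; Destructible planter quest prop for a Battle Pass Season 29 Snapshot quest, built on a quest prop parent class.
; On server damage, it triggers destruction cosmetics only if the instigating player has not completed the quest, preventing repeat FX, then calls parent damage handling.
; The cosmetic event spawns sound and offset Niagara FX only on non-authority machines where a helper deems the instigator locally controlled, with a split-screen controller check.
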(event EventBeginPlay
  (|CallParentFunction) ; This node is disabled and will not be called.\nDrag off pins to build functionality.
)

(event Collision|EventActorBeginOverlap
  (params (OtherActor :type Actor%20Object%20Reference)))

(event EventTick
  (params (DeltaSeconds :type Float%20%28single-precision%29)))

(event EventOnDamageServer
  (params (Damage :type Float%20%28single-precision%29) (DamageTags :type Gameplay%20Tag%20Container%20Structure%20%28by%20ref%29) (Momentum :type Vector :default "0, 0, 0") (HitInfo :type Hit%20Result%20Structure%20%28by%20ref%29) (InstigatedBy :type Controller%20Object%20Reference) (DamageCauser :type Actor%20Object%20Reference) (EffectContext :type Gameplay%20Effect%20Context%20Handle%20Structure))
  ;[0 0 0] Play custom FX for when the planter is destroyed to prevent it from playing twice
  (Utilities|FlowControl|Sequence
    (:then_0
      (bind _asfort_player_controller (Utilities|Casting|CastToFortPlayerController InstigatedBy))
      (bind (_playercontrollerout _completedquest) (Quest|HasPlayerCompletedQuest _asfort_player_controller (Class|ConsolidatedQuestComponent|GetQuestItemDefinition (Variables|BuildingGameplayActorPropQuest|GetConsolidatedQuestComponent)) (Variables|/Script/FortniteGame.ConsolidatedQuestComponent|GetObjectiveBackendName (Variables|BuildingGameplayActorPropQuest|GetConsolidatedQuestComponent))))
      (if (not _completedquest)
        (CallFunction|HandleDestructionCosmeticEvents _playercontrollerout)))
    (:then_1
      (|CallParentFunction Damage DamageTags Momentum HitInfo InstigatedBy DamageCauser EffectContext))))

(event Custom|HandleDestructionCosmeticEvents
  (params (InstigatingPlayerController :type Fort%20Player%20Controller%20Object%20Reference))
  (bind _returnvalue (Transformation|GetActorLocation))
  ;[0 0 0] Play custom FX for when the planter is destroyed to prevent it from playing twice
  (Networking|SwitchHasAuthority
    (:Remote
      (bind _output (CallFunction|IsInstigatorLocallyControlled InstigatingPlayerController))
      (if _output
        (Audio|SpawnSoundatLocation (Variables|PlanterDestruction|GetDestructionAudio) _returnvalue)
        (Niagara|SpawnSystematLocation (Variables|PlanterDestruction|GetDestructionFX) (+ _returnvalue (.location (Variables|Design|PostInteraction|AmbientFX|GetPostInteractionAmbientParticleTransform))))))))

(fn IsInstigatorLocallyControlled
  (params (InstigatingPlayerController :type Fort%20Player%20Controller%20Object%20Reference))
  ;[0 0 0] Returns true on the executing client if the instigator controller is locally controlled. Returns false if running on the server
  (if (Networking|IsServer)
    (return false)
    (else
      (Utilities|IsValid InstigatingPlayerController
        (:Is%20Valid
          (if (PlayerController|IsinSplitscreen InstigatingPlayerController)
            (for _array_element (Fortnite|GetLocalFortPlayerControllers true)
              (if (== _array_element InstigatingPlayerController)
                (return true)))
            (return false)
            (else
              (return true))))))))

(fn ConstructionScript
  (|CallParentFunction))
